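# Niagara swing trail effect for the blue Samurai pickaxe, built from four short-lived ribbon emitters.
# Blade, hilt and handle emitters use AnimTrail modules to stretch ribbons between socket pairs on the User.SkeletalMesh input.
# Mask_Tube adds a rate-spawned ribbon at a mesh bone/socket and writes a dynamic material parameter.
# All emitters spawn only while the system Timeline is not playing, and fade particle alpha by curve.
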NiagaraSystem /BRCosmetics/Weapons/FORT_Melee/Pickaxe_Samurai_Blue/FX/NS_SamuraiBlue_Swing
dependencies {
 usedDataInterfaces [&NiagaraDataInterfaceCurve]
 usedDynamicInputs []
 usedModules [&SystemState, &Timeline]
}
emitters {
 Mask_Tube {
  emitterUpdateScript [
   &EmitterState(Life Cycle Mode=::NewEnumerator1, Loop Behavior=::NewEnumerator1, Loop Duration=0.333, UseLoopDelay=false)
   &SpawnRate(
    SpawnRate=&MakeCustomFloatFromBool(Bool=&Boolean_NotOperation(Bool=$System.Timeline.IsPlaying), True Float=60, False Float=0)
    Use Spawn Probability=false
   )
   &SpawnPerUnit(
    Max Movement Threshold=15000
    Module.Use Max Movement Threshold=true
    Module.Use Movement Tolerance=false
    Spawn Probability:NiagaraFloat=&MakeCustomFloatFromBool(Bool=&Boolean_NotOperation(Bool=$System.Timeline.IsPlaying), True Float=1, False Float=0)
    Spawn Spacing=1
    Use Spawn Probability=true
   ) {enabled=false}
  ]
  particleSpawnScript [
   &InitializeParticle(
    Color=<rgba 1 1 1 0.5>
    Lifetime=0.075
    Ribbon Width=15
    Ribbon Width Mode=::NewEnumerator1
    UsePositionOffset=false
    Write Material Random=true
   )
   &SkeletalMeshLocation(
    Bone / Socket Index=2
    MaskSpawnGroup=false
    Mesh Sampling Type=::NewEnumerator5
    Skeletal Mesh=$User.SkeletalMesh
    Socket Sampling Mode=::NewEnumerator5
   )
  ]
  particleUpdateScript [
   &ParticleState()
   &ScaleColor(
    Scale Alpha=&FloatFromCurve(
     FloatCurve=DataInterface() {propertyValues {Curve {keys [<at 0.15 1 RCIM_Cubic>, {interpMode=RCIM_Cubic, leaveTangent=-2.1717, time=1, value=0}]}}}
     Scale Curve=3
    )
    ScaleA=true
    ScaleRGB=true
   )
   &DynamicMaterialParameters(Module.Param0WriteEnabled=true, Module.Param1WriteEnabled=false, Module.Param2WriteEnabled=false, Module.Param3WriteEnabled=false)
  ]
  rendererClasses [&NiagaraRibbonRendererProperties]
  renderers [{rendererClass=&NiagaraRibbonRendererProperties}]
 }
 Wyeth_AnimTrails_Blade {
  emitterUpdateScript [
   &EmitterState(Life Cycle Mode=::NewEnumerator1, Loop Behavior=::NewEnumerator1, Loop Duration=0.333, UseLoopDelay=false)
   &self:SpawnAnimTrail(
    AnimTrail is Active:NiagaraBool=&Boolean_NotOperation(Bool=$System.Timeline.IsPlaying)
    Mesh:NiagaraDataInterfaceSkeletalMesh=$User.SkeletalMesh
    Socket B=1
   ) {module=SpawnAnimTrail}
  ]
  particleSpawnScript [&self:InitializeAnimTrail(Color:LinearColor=<rgba 0.7177 0.7943 0.8385 0.2>, Lifetime=0.075, Ribbon Width Scale=1) {module=InitializeAnimTrail}]
  particleUpdateScript [
   &ParticleState() {module=UpdateAge}
   &SolveForcesAndVelocity(Clamp Velocity=false, Limit Acceleration=false)
   &self:UpdateAnimTrail() {module=UpdateAnimTrail}
   &ScaleColor(
    Scale Alpha=&FloatFromCurve(FloatCurve=DataInterface() {propertyValues {Curve {keys [<at 0.15 1 RCIM_Cubic>, {interpMode=RCIM_Cubic, leaveTangent=-2.1717, time=1, value=0}]}}})
    ScaleA=true
    ScaleRGB=true
   )
  ]
  rendererClasses [&NiagaraRibbonRendererProperties]
  renderers [{rendererClass=&NiagaraRibbonRendererProperties}]
 }
 Wyeth_AnimTrails_Handle {
  emitterUpdateScript [
   &EmitterState(Life Cycle Mode=::NewEnumerator1, Loop Behavior=::NewEnumerator1, Loop Duration=0.333, UseLoopDelay=false)
   &self:SpawnAnimTrail(
    AnimTrail is Active:NiagaraBool=&Boolean_NotOperation(Bool=$System.Timeline.IsPlaying)
    Mesh:NiagaraDataInterfaceSkeletalMesh=$User.SkeletalMesh
    Socket A=2
    Socket B=3
   ) {module=SpawnAnimTrail}
  ]
  particleSpawnScript [&self:InitializeAnimTrail(Color:LinearColor=<rgba 1 1 1 0.375>, Lifetime=0.075, Ribbon Width Scale=1) {module=InitializeAnimTrail}]
  particleUpdateScript [
   &ParticleState() {module=UpdateAge}
   &SolveForcesAndVelocity(Clamp Velocity=false, Limit Acceleration=false)
   &self:UpdateAnimTrail() {module=UpdateAnimTrail}
   &ScaleColor(
    Scale Alpha=&FloatFromCurve(FloatCurve=DataInterface() {propertyValues {Curve {keys [<at 0.15 1 RCIM_Cubic>, {interpMode=RCIM_Cubic, leaveTangent=-2.1717, time=1, value=0}]}}})
    ScaleA=true
    ScaleRGB=true
   )
  ]
  rendererClasses [&NiagaraRibbonRendererProperties]
  renderers [{rendererClass=&NiagaraRibbonRendererProperties}]
 }
 Wyeth_AnimTrails_Hilt {
  emitterUpdateScript [
   &EmitterState(Life Cycle Mode=::NewEnumerator1, Loop Behavior=::NewEnumerator1, Loop Duration=0.333, UseLoopDelay=false)
   &self:SpawnAnimTrail(
    AnimTrail is Active:NiagaraBool=&Boolean_NotOperation(Bool=$System.Timeline.IsPlaying)
    Mesh:NiagaraDataInterfaceSkeletalMesh=$User.SkeletalMesh
    Socket A=1
    Socket B=2
   ) {module=SpawnAnimTrail}
  ]
  particleSpawnScript [&self:InitializeAnimTrail(Color:LinearColor=<rgba 1 1 1 0.75>, Lifetime=0.075, Ribbon Width Scale=1) {module=InitializeAnimTrail}]
  particleUpdateScript [
   &ParticleState() {module=UpdateAge}
   &SolveForcesAndVelocity(Clamp Velocity=false, Limit Acceleration=false)
   &self:UpdateAnimTrail() {module=UpdateAnimTrail}
   &ScaleColor(
    Scale Alpha=&FloatFromCurve(FloatCurve=DataInterface() {propertyValues {Curve {keys [<at 0.15 1 RCIM_Cubic>, {interpMode=RCIM_Cubic, leaveTangent=-2.1717, time=1, value=0}]}}})
    ScaleA=true
    ScaleRGB=true
   )
  ]
  rendererClasses [&NiagaraRibbonRendererProperties]
  renderers [{rendererClass=&NiagaraRibbonRendererProperties}]
 }
}
summary {
 systemName=NS_SamuraiBlue_Swing
 emitters [
  {emitterName=Wyeth_AnimTrails_Blade, rendererClasses [&NiagaraRibbonRendererProperties]}
  {emitterName=Wyeth_AnimTrails_Hilt, rendererClasses [&NiagaraRibbonRendererProperties]}
  {emitterName=Wyeth_AnimTrails_Handle, rendererClasses [&NiagaraRibbonRendererProperties]}
  {emitterName=Mask_Tube, rendererClasses [&NiagaraRibbonRendererProperties]}
 ]
 userVariables [User.SkeletalMesh {
   type=NiagaraDataInterfaceSkeletalMesh
   defaultValue {
    struct=NiagaraExt_VariableValue_DataInterface
    value {dataInterface=&self:NiagaraDataInterfaceSkeletalMesh_0}
   }
  }]
}
