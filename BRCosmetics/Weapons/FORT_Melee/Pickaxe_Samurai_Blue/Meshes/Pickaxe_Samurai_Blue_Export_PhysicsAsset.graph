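# Physics asset for the Pickaxe_Samurai_Blue two-handed melee pickaxe, setting up its dangling rope and two straps for physics.
# A long capsule covers the handle, and small capsules cover the rope and each strap bone.
# Each dyn_ bone is constrained directly to the handle rather than chained to another.
# The straps lock twist and one swing axis so they hinge in a single plane, while the rope leaves those axes unlocked.
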PhysicsAsset /BRCosmetics/Weapons/FORT_Melee/Pickaxe_Samurai_Blue/Meshes/Pickaxe_Samurai_Blue_Export_PhysicsAsset
bodies {
 Melee_TwoHanded_Handle [{shapeType=Capsule, length=120.7899, radius=1.7725, center=<0.5612 -1.9365 36.5565>, rotation=<rot 1.2512 -90 -0.1336>}]
 dyn_rope [{shapeType=Capsule, length=4.0703, radius=4.0645, center=<0.9842 -2.2086 -5.2042>, rotation=<rot 8.1776 5.7753 35.4139>}]
 dyn_strapA [{shapeType=Capsule, length=2.3419, radius=1.6527, center=<3.1914 0 0.2039>, rotation=<rot -90 -0.0074 0.0074>}]
 dyn_strapB [{shapeType=Capsule, length=1.6005, radius=1.3999, center=<2.6948 0 0>, rotation=<rot -90 -0.0297 0.0297>}]
}
constraints [
 {bone1Name=dyn_rope, bone2Name=Melee_TwoHanded_Handle, swing1LimitDegrees=45, swing2LimitDegrees=45, twistLimitDegrees=45}
 {
  bone1Name=dyn_strapB
  bone2Name=Melee_TwoHanded_Handle
  swing1LimitDegrees=90
  swing2LimitDegrees=45
  swing2Motion=Locked
  twistLimitDegrees=45
  twistMotion=Locked
 }
 {
  bone1Name=dyn_strapA
  bone2Name=Melee_TwoHanded_Handle
  swing1LimitDegrees=90
  swing2LimitDegrees=45
  swing2Motion=Locked
  twistLimitDegrees=45
  twistMotion=Locked
 }
]
